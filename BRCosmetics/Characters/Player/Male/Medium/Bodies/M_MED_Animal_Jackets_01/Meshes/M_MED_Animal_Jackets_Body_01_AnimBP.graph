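# Post-style cosmetic AnimBP for the male medium Animal Jackets body, adding secondary physics motion to the coat bones.
# Copies the pose from the attached parent mesh, then chains nine partially blended AnimDynamics nodes in component space, limited by LOD threshold.
# Nodes share heavy damping, reduced gravity and angular limits, with left-side joint offsets mirrored.
# dyn_coat_bl is simulated twice and no dyn_coat_fl node exists; the EventGraph has no logic.
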AnimBlueprint /BRCosmetics/Characters/Player/Male/Medium/Bodies/M_MED_Animal_Jackets_01/Meshes/M_MED_Animal_Jackets_Body_01_AnimBP
target_skeleton=&/M_MED_Animal_Jackets_Body_01_Skeleton
graphs [
 AnimGraph {nodes [
   n1 = Anim Dynamics - Bone: dyn_coat_back(ComponentPose=Local To Component(LocalPose=Copy Pose From Mesh() {bUseAttachedParent=true}), Alpha=0.75, ExternalForce="0,0,0") {
    angularDampingOverride=0.98
    bGravityOverrideInSimSpace=true
    bOverrideAngularDamping=true
    bOverrideLinearDamping=true
    gravityScale=0.35
    lODThreshold=1
    linearDampingOverride=0.98
    numSolverIterationsPostUpdate=4
    numSolverIterationsPreUpdate=16
    alphaBoolBlend {blendOption=Linear}
    boundBone {boneName=dyn_coat_back}
    physicsBodyDefinitions [{
      sphereCollisionRadius=0
      boundBone {boneName=dyn_coat_back}
      boxExtents=<5 5 5>
      constraintSetup {angularLimitsMax=<5 15 15>, angularLimitsMin=<-45 -15 -15>, angularTarget=<0 0 0>}
      localJointOffset=<0 0 -0.5>
     }]
   }
   n2 = Anim Dynamics - Bone: dyn_coat_fr(ComponentPose=@n1, Alpha=0.75, ExternalForce="0,0,0") {
    angularDampingOverride=0.98
    bGravityOverrideInSimSpace=true
    bOverrideAngularDamping=true
    bOverrideLinearDamping=true
    gravityScale=0.35
    lODThreshold=1
    linearDampingOverride=0.98
    numSolverIterationsPostUpdate=4
    numSolverIterationsPreUpdate=16
    alphaBoolBlend {blendOption=Linear}
    boundBone {boneName=dyn_coat_fr}
    physicsBodyDefinitions [{
      sphereCollisionRadius=0
      boundBone {boneName=dyn_coat_fr}
      boxExtents=<5 5 5>
      constraintSetup {angularLimitsMax=<5 15 15>, angularLimitsMin=<-45 -15 -15>, angularTarget=<0 0 0>}
      localJointOffset=<0 0 -0.5>
     }]
   }
   n3 = Anim Dynamics - Bone: dyn_coat_br(ComponentPose=@n2, Alpha=0.75, ExternalForce="0,0,0") {
    angularDampingOverride=0.98
    bGravityOverrideInSimSpace=true
    bOverrideAngularDamping=true
    bOverrideLinearDamping=true
    gravityScale=0.35
    lODThreshold=1
    linearDampingOverride=0.98
    numSolverIterationsPostUpdate=4
    numSolverIterationsPreUpdate=16
    alphaBoolBlend {blendOption=Linear}
    boundBone {boneName=dyn_coat_br}
    physicsBodyDefinitions [{
      sphereCollisionRadius=0
      boundBone {boneName=dyn_coat_br}
      boxExtents=<5 5 5>
      constraintSetup {angularLimitsMax=<5 15 15>, angularLimitsMin=<-45 -15 -15>, angularTarget=<0 0 0>}
      localJointOffset=<0 0 -0.5>
     }]
   }
   n4 = Anim Dynamics - Bone: dyn_coat_r(ComponentPose=@n3, Alpha=0.75, ExternalForce="0,0,0") {
    angularDampingOverride=0.98
    bGravityOverrideInSimSpace=true
    bOverrideAngularDamping=true
    bOverrideLinearDamping=true
    gravityScale=0.35
    lODThreshold=1
    linearDampingOverride=0.98
    numSolverIterationsPostUpdate=4
    numSolverIterationsPreUpdate=16
    alphaBoolBlend {blendOption=Linear}
    boundBone {boneName=dyn_coat_r}
    physicsBodyDefinitions [{
      sphereCollisionRadius=0
      boundBone {boneName=dyn_coat_r}
      boxExtents=<5 5 5>
      constraintSetup {angularLimitsMax=<5 15 15>, angularLimitsMin=<-45 -15 -15>, angularTarget=<0 0 0>}
      localJointOffset=<0 0 -0.5>
     }]
   }
   n5 = Anim Dynamics - Bone: dyn_coat_front_r(ComponentPose=@n4, Alpha=0.75, ExternalForce="0,0,0") {
    angularDampingOverride=0.98
    bGravityOverrideInSimSpace=true
    bOverrideAngularDamping=true
    bOverrideLinearDamping=true
    gravityScale=0.35
    lODThreshold=1
    linearDampingOverride=0.98
    numSolverIterationsPostUpdate=4
    numSolverIterationsPreUpdate=16
    alphaBoolBlend {blendOption=Linear}
    boundBone {boneName=dyn_coat_front_r}
    physicsBodyDefinitions [{
      sphereCollisionRadius=0
      boundBone {boneName=dyn_coat_front_r}
      boxExtents=<5 5 5>
      constraintSetup {angularLimitsMax=<5 15 15>, angularLimitsMin=<-45 -15 -15>, angularTarget=<0 0 0>}
      localJointOffset=<0 0 -0.5>
     }]
   }
   n6 = Anim Dynamics - Bone: dyn_coat_bl(ComponentPose=@n5, Alpha=0.75, ExternalForce="0,0,0") {
    angularDampingOverride=0.98
    bGravityOverrideInSimSpace=true
    bOverrideAngularDamping=true
    bOverrideLinearDamping=true
    gravityScale=0.35
    lODThreshold=1
    linearDampingOverride=0.98
    numSolverIterationsPostUpdate=4
    numSolverIterationsPreUpdate=16
    alphaBoolBlend {blendOption=Linear}
    boundBone {boneName=dyn_coat_bl}
    physicsBodyDefinitions [{
      sphereCollisionRadius=0
      boundBone {boneName=dyn_coat_bl}
      boxExtents=<5 5 5>
      constraintSetup {angularLimitsMax=<5 15 15>, angularLimitsMin=<-45 -15 -15>, angularTarget=<0 0 0>}
      localJointOffset=<0 0 0.5>
     }]
   }
   n7 = Anim Dynamics - Bone: dyn_coat_bl(ComponentPose=@n6, Alpha=0.75, ExternalForce="0,0,0") {
    angularDampingOverride=0.98
    bGravityOverrideInSimSpace=true
    bOverrideAngularDamping=true
    bOverrideLinearDamping=true
    gravityScale=0.35
    lODThreshold=1
    linearDampingOverride=0.98
    numSolverIterationsPostUpdate=4
    numSolverIterationsPreUpdate=16
    alphaBoolBlend {blendOption=Linear}
    boundBone {boneName=dyn_coat_bl}
    physicsBodyDefinitions [{
      sphereCollisionRadius=0
      boundBone {boneName=dyn_coat_bl}
      boxExtents=<5 5 5>
      constraintSetup {angularLimitsMax=<5 15 15>, angularLimitsMin=<-45 -15 -15>, angularTarget=<0 0 0>}
      localJointOffset=<0 0 0.5>
     }]
   }
   n8 = Anim Dynamics - Bone: dyn_coat_l(ComponentPose=@n7, Alpha=0.75, ExternalForce="0,0,0") {
    angularDampingOverride=0.98
    bGravityOverrideInSimSpace=true
    bOverrideAngularDamping=true
    bOverrideLinearDamping=true
    gravityScale=0.35
    lODThreshold=1
    linearDampingOverride=0.98
    numSolverIterationsPostUpdate=4
    numSolverIterationsPreUpdate=16
    alphaBoolBlend {blendOption=Linear}
    boundBone {boneName=dyn_coat_l}
    physicsBodyDefinitions [{
      sphereCollisionRadius=0
      boundBone {boneName=dyn_coat_l}
      boxExtents=<5 5 5>
      constraintSetup {angularLimitsMax=<5 15 15>, angularLimitsMin=<-45 -15 -15>, angularTarget=<0 0 0>}
      localJointOffset=<0 0 0.5>
     }]
   }
   n9 = Anim Dynamics - Bone: dyn_coat_front_l(ComponentPose=@n8, Alpha=0.75, ExternalForce="0,0,0") {
    angularDampingOverride=0.98
    bGravityOverrideInSimSpace=true
    bOverrideAngularDamping=true
    bOverrideLinearDamping=true
    gravityScale=0.35
    lODThreshold=1
    linearDampingOverride=0.98
    numSolverIterationsPostUpdate=4
    numSolverIterationsPreUpdate=16
    alphaBoolBlend {blendOption=Linear}
    boundBone {boneName=dyn_coat_front_l}
    physicsBodyDefinitions [{
      sphereCollisionRadius=0
      boundBone {boneName=dyn_coat_front_l}
      boxExtents=<5 5 5>
      constraintSetup {angularLimitsMax=<5 15 15>, angularLimitsMin=<-45 -15 -15>, angularTarget=<0 0 0>}
      localJointOffset=<0 0 0.5>
     }]
   }
   Output Pose(Result=Component To Local(ComponentPose=@n9)) {layerGroup=None}
  ]}
 EventGraph {nodes [Event BlueprintUpdateAnimation()]}
]
variables []
