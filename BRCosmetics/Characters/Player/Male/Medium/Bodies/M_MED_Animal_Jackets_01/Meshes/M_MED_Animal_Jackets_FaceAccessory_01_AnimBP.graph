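# Animation Blueprint for the Animal Jackets face accessory, adding secondary motion to ears and a tag atop the pose copied from the attached parent mesh.
# Spring-driven AnimDynamics move each ear, switching simulation space with pawn speed and lowering alpha while targeting.
# A chained AnimDynamics swings the ring and tag, then the pose is cached.
# When the playlist uses custom character parts, checked at initialization, output shrinks the head bone to near zero.
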AnimBlueprint /BRCosmetics/Characters/Player/Male/Medium/Bodies/M_MED_Animal_Jackets_01/Meshes/M_MED_Animal_Jackets_FaceAccessory_01_AnimBP
target_skeleton=&/M_MED_Animal_Jackets_FaceAccessory_01_Skeleton
graphs [
 AnimGraph {nodes [
   n1 = "Transform (Modify) Bone - Bone: head"!(ComponentPose=Local To Component(LocalPose=Use cached pose 'PreScale Cache'()), Alpha=1) -> (Pose) {scaleMode=BMM_Replace, alphaBoolBlend {blendOption=Linear}, boneToModify {boneName=head}, scale=<0.01 0.01 0.01>}
   Output Pose(Result=Blend Poses by bool(
     BlendPose_0=Component To Local(ComponentPose=@n1)
     BlendPose_1=Use cached pose 'PreScale Cache'()
     BlendTime_0=0
     BlendTime_1=0
     bActiveValue=Get GetPlayerListUsesCustomCharacterParts() -> (GetPlayerListUsesCustomCharacterParts:bool)
    ) {blendType=Linear}) {layerGroup=None}
   n2 = SelectFloat(A=0.15, B=1, bPickA=Get bIsTargeting())
   n3 = Select(
    Option 0:AnimPhysSimSpaceType Enum=BoneRelative
    Option 1:AnimPhysSimSpaceType Enum=Component
    Index="float > float"(A=Get PawnSpeed2D(), B=0) {is=title K2Node_CallFunction}
   ) -> (ReturnValue:AnimPhysSimSpaceType Enum)
   n4 = Anim Dynamics - Bone: dyn_ear_r!(
    SimulationSpace:AnimPhysSimSpaceType Enum=@n3
    ComponentPose=Local To Component(LocalPose=Copy Pose From Mesh() {bUseAttachedParent=true})
    Alpha=@n2
    ExternalForce="0,0,0"
   ) -> (Pose) {
    angularDampingOverride=0.98
    angularSpringConstant=275
    bAngularSpring=true
    bGravityOverrideInSimSpace=true
    bLinearSpring=true
    bOverrideAngularDamping=true
    bOverrideLinearDamping=true
    gravityScale=0.25
    lODThreshold=1
    linearDampingOverride=0.98
    linearSpringConstant=275
    simulationSpace=BoneRelative
    alphaBoolBlend {blendOption=Linear}
    alphaScaleBiasClamp {
     bClampResult=true
     bInterpResult=true
     clampMax=0.25
     clampMin=0.05
     interpSpeedDecreasing=0.5
     interpSpeedIncreasing=0.5
     outRange {max=0.1}
    }
    boundBone {boneName=dyn_ear_r}
    physicsBodyDefinitions [{
      sphereCollisionRadius=0
      boundBone {boneName=dyn_ear_r}
      boxExtents=<30 30 30>
      constraintSetup {
       angularTargetAxis=AxisY
       twistAxis=AxisZ
       angularLimitsMax=<45 45 45>
       angularLimitsMin=<-45 -45 -45>
       angularTarget=<0 0 0>
       linearAxesMax=<10 10 10>
       linearAxesMin=<-10 -10 -10>
      }
      localJointOffset=<0 0 0>
     }]
    relativeSpaceBone {boneName=head}
   }
   n5 = Anim Dynamics - Bone: dyn_ear_l!(SimulationSpace:AnimPhysSimSpaceType Enum=@n3, ComponentPose=@n4, Alpha=@n2, ExternalForce="0,0,0") -> (Pose) {
    angularDampingOverride=0.98
    angularSpringConstant=260
    bAngularSpring=true
    bGravityOverrideInSimSpace=true
    bLinearSpring=true
    bOverrideAngularDamping=true
    bOverrideLinearDamping=true
    gravityScale=0.2
    lODThreshold=1
    linearDampingOverride=0.98
    linearSpringConstant=260
    simulationSpace=BoneRelative
    alphaBoolBlend {blendOption=Linear}
    alphaScaleBiasClamp {
     bClampResult=true
     bInterpResult=true
     clampMax=0.25
     clampMin=0.05
     interpSpeedDecreasing=0.5
     interpSpeedIncreasing=0.5
     outRange {max=0.1}
    }
    boundBone {boneName=dyn_ear_l}
    physicsBodyDefinitions [{
      sphereCollisionRadius=0
      boundBone {boneName=dyn_ear_l}
      boxExtents=<30 30 30>
      constraintSetup {
       angularTargetAxis=AxisY
       twistAxis=AxisZ
       angularLimitsMax=<45 45 45>
       angularLimitsMin=<-45 -45 -45>
       angularTarget=<0 0 0>
       linearAxesMax=<10 10 10>
       linearAxesMin=<-10 -10 -10>
      }
      localJointOffset=<0 0 0>
     }]
    relativeSpaceBone {boneName=head}
   }
   n6 = "Anim Dynamics - Chain: dyn_ring -> dyn_tag"(ComponentPose=@n5, Alpha=1, ExternalForce="0,0,0") {
    bChain=true
    bGravityOverrideInSimSpace=true
    lODThreshold=1
    numSolverIterationsPostUpdate=4
    numSolverIterationsPreUpdate=16
    alphaBoolBlend {blendOption=Linear}
    boundBone {boneName=dyn_ring}
    chainEnd {boneName=dyn_tag}
    physicsBodyDefinitions [{
      sphereCollisionRadius=0
      boundBone {boneName=dyn_ring}
      boxExtents=<5 5 5>
      constraintSetup {angularTargetAxis=AxisZ, twistAxis=AxisZ, angularLimitsMax=<5 30 30>, angularLimitsMin=<-90 -30 -30>, angularTarget=<0 0 0>}
      localJointOffset=<0 0 -0.5>
     }]
   }
   Save cached pose 'PreScale Cache'(Pose=Component To Local(ComponentPose=@n6)) {cacheName=PreScale Cache}
  ]}
 EventGraph {nodes [
   Event BlueprintInitializeAnimation()
   > Set GetPlayerListUsesCustomCharacterParts(GetPlayerListUsesCustomCharacterParts:bool=GetPlaylistUsesCustomCharacterParts!() -> (ReturnValue:bool)) -> (Output_Get:bool)
   Event BlueprintUpdateAnimation()
  ]}
]
variables [GetPlayerListUsesCustomCharacterParts {type=boolean}]
